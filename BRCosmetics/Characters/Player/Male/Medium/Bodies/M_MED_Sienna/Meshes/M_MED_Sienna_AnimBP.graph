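# Post-process style Animation Blueprint for the M_MED_Sienna player body, adding corrective deformation and cape secondary motion to the pose copied from its parent mesh.
# Twist and sculpter Control Rigs feed constraints and variable-weighted bone offsets for straps, pouches, collar and cape.
# State-weighted RigidBody nodes simulate the cape across backpack, frontend and emote cases.
# An AnimDynamics chain extends that cached pose, and bRigidBodyEnabled selects between both results.
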AnimBlueprint /BRCosmetics/Characters/Player/Male/Medium/Bodies/M_MED_Sienna/Meshes/M_MED_Sienna_AnimBP
target_skeleton=&/M_MED_Sienna_Skeleton
graphs [
 AnimGraph {nodes [
   Anim Dynamics - Bone: cape_01_l(Alpha=1, ExternalForce="0,0,0") {
    angularDampingOverride=1
    angularSpringConstant=10000
    bAngularSpring=true
    bOverrideAngularDamping=true
    bUseGravityOverride=true
    gravityScale=0.35
    simulationSpace=BoneRelative
    alphaBoolBlend {blendOption=Linear}
    boundBone {boneName=cape_01_l}
    componentAppliedLinearAccClamp=<3000 3000 3000>
    componentLinearAccScale=<2 2 2>
    componentLinearVelScale=<2 2 2>
    gravityOverride {z=-980}
    physicsBodyDefinitions [{
      boundBone {boneName=cape_01_l}
      boxExtents=<10 40 40>
      constraintSetup {angularLimitsMax=<0 20 0>, angularLimitsMin=<0 -20 -15>, angularTarget=<1 0 0>}
      localJointOffset=<5 0 0>
     }]
    relativeSpaceBone {boneName=backpack_root}
   }
   Anim Dynamics - Bone: cape_01_l(Alpha=1, ExternalForce="0,0,0") {
    angularDampingOverride=1
    angularSpringConstant=10000
    bAngularSpring=true
    bOverrideAngularDamping=true
    bUseGravityOverride=true
    gravityScale=0.35
    simulationSpace=World
    alphaBoolBlend {blendOption=Linear}
    boundBone {boneName=cape_01_l}
    componentAppliedLinearAccClamp=<3000 3000 3000>
    componentLinearAccScale=<2 2 2>
    componentLinearVelScale=<2 2 2>
    gravityOverride {z=-980}
    physicsBodyDefinitions [{
      boundBone {boneName=cape_01_l}
      boxExtents=<30 40 40>
      constraintSetup {angularLimitsMax=<0 5 5>, angularLimitsMin=<0 -5 -15>, angularTarget=<1 0 0>}
      localJointOffset=<4 0 0>
     }]
    relativeSpaceBone {boneName=backpack_root}
   }
   n1 = "Constraint - cape_00_l (TT)"!(ComponentPose, Alpha=1, ConstraintWeights_0=0.25, ConstraintWeights_1:float=0.75) -> (Pose) {
    alphaBoolBlend {blendOption=Linear}
    boneToModify {boneName=cape_00_l}
    constraintSetup [{transformType=Translation, targetBone {boneName=clavicle_l}}, {transformType=Translation, targetBone {boneName=spine_05}}]
    constraintWeights [1,1]
   }
   "Constraint - cape_00_r (TT)"!(ComponentPose=@n1, Alpha=1, ConstraintWeights_0=0.5, ConstraintWeights_1:float=0.5) -> (Pose) {
    alphaBoolBlend {blendOption=Linear}
    boneToModify {boneName=cape_00_r}
    constraintSetup [{transformType=Translation, targetBone {boneName=clavicle_r}}, {transformType=Translation, targetBone {boneName=spine_05}}]
    constraintWeights [1,1]
   }
   n2 = "Transform (Modify) Bone - Bone: cape_00_l"(Alpha=1, Translation="0,0,0", Rotation="0,0,-10", Scale="1,1,1") {
    rotationMode=BMM_Additive
    rotationSpace=BCS_BoneSpace
    scaleMode=BMM_Replace
    scaleSpace=BCS_BoneSpace
    translationMode=BMM_Additive
    translationSpace=BCS_BoneSpace
    alphaBoolBlend {blendOption=Linear}
    boneToModify {boneName=cape_00_l}
    rotation {roll=-10}
   }
   "Transform (Modify) Bone - Bone: cape_00_r"(ComponentPose=@n2, Alpha=1, Translation="0,0,0", Rotation="0,0,-10", Scale="1,1,1") {
    rotationMode=BMM_Additive
    rotationSpace=BCS_BoneSpace
    scaleMode=BMM_Replace
    scaleSpace=BCS_BoneSpace
    translationMode=BMM_Additive
    translationSpace=BCS_BoneSpace
    alphaBoolBlend {blendOption=Linear}
    boneToModify {boneName=cape_00_r}
    rotation {roll=-10}
   }
   Output Pose(Result=Blend Poses by bool(BlendPose_0=Use cached pose 'previous setup'(), BlendPose_1=Use cached pose 'animDyn Cache'(), bActiveValue=Get bRigidBodyEnabled()) {blendType=Linear}) {layerGroup=None}
   n3 = Control Rig(Source=Copy Pose From Mesh() {bUseAttachedParent=true}) {
    alphaBoolBlend {blendOption=Linear}
    controlRigAssetReference {blueprintRigClass=&/M_MED_Sienna_CtrlRig_Twists.M_MED_Sienna_CtrlRig_Twists_C}
   }
   n4 = Control Rig!(
    Source=@n3
    Alpha
    KneeRot_fr_l_114deg:Vector
    KneeTrans_fr_l_114deg:Vector="0, 0, 0"
    KneeScale_fr_l_114deg:Vector="0, 0, 0"
    KneeScale_fr_l_130deg:Vector="0, 0, 0"
    KneeRot_fr_l_130deg:Vector="0, 0, 0"
    KneeTrans_fr_l_130deg:Vector="0, 0, 0"
   ) -> (Pose) {
    alphaBoolBlend {blendOption=Linear}
    controlRigAssetReference {blueprintRigClass=&/M_MED_Sienna_CtrlRig_Sculpters.M_MED_Sienna_CtrlRig_Sculpters_C}
   }
   n5 = "Constraint - deform_trap_l (TT)"!(ComponentPose=Local To Component(LocalPose=@n4), Alpha=1, ConstraintWeights_0=0.5, ConstraintWeights_1:float=0.5) -> (Pose) {
    alphaBoolBlend {blendOption=Linear}
    boneToModify {boneName=deform_trap_l}
    constraintSetup [{transformType=Translation, targetBone {boneName=upperarm_l}}, {transformType=Translation, targetBone {boneName=neck_02}}]
    constraintWeights [1,1]
   }
   n6 = "Constraint - deform_trap_r (TT)"!(ComponentPose=@n5, Alpha=1, ConstraintWeights_0=0.5, ConstraintWeights_1:float=0.5) -> (Pose) {
    alphaBoolBlend {blendOption=Linear}
    boneToModify {boneName=deform_trap_r}
    constraintSetup [{transformType=Translation, targetBone {boneName=upperarm_r}}, {transformType=Translation, targetBone {boneName=neck_02}}]
    constraintWeights [1,1]
   }
   n7 = "Transform (Modify) Bone - Bone: deform_trap_l"(ComponentPose=@n6, Alpha=Get clavLf() -> (clavLf:double), Translation="3,0,3", Rotation="0,0,0", Scale="1,1,1") {
    rotationMode=BMM_Additive
    rotationSpace=BCS_BoneSpace
    scaleMode=BMM_Replace
    scaleSpace=BCS_BoneSpace
    translationMode=BMM_Additive
    translationSpace=BCS_BoneSpace
    alphaBoolBlend {blendOption=Linear}
    boneToModify {boneName=deform_trap_l}
    translation {x=3, z=3}
   }
   n8 = "Transform (Modify) Bone - Bone: deform_trap_r"(ComponentPose=@n7, Alpha=Get clavRt() -> (clavRt:double), Translation="3,0,3", Rotation="0,0,0", Scale="1,1,1") {
    rotationMode=BMM_Additive
    rotationSpace=BCS_BoneSpace
    scaleMode=BMM_Replace
    scaleSpace=BCS_BoneSpace
    translationMode=BMM_Additive
    translationSpace=BCS_BoneSpace
    alphaBoolBlend {blendOption=Linear}
    boneToModify {boneName=deform_trap_r}
    translation {x=3, z=3}
   }
   n9 = "Transform (Modify) Bone - Bone: StrapPOSING"(ComponentPose=@n8, Alpha=Get LegFwd() -> (LegFwd:double), Translation="0,0,0", Rotation="-30,-20,0", Scale="1,1,1") {
    rotationMode=BMM_Additive
    rotationSpace=BCS_BoneSpace
    scaleMode=BMM_Replace
    scaleSpace=BCS_BoneSpace
    translationMode=BMM_Additive
    translationSpace=BCS_BoneSpace
    alphaBoolBlend {blendOption=Linear}
    boneToModify {boneName=StrapPOSING}
    rotation {pitch=-30, yaw=-20}
   }
   n10 = "Constraint - pouch1_l (PPPPPPPPPPP)"!(
    ComponentPose=@n9
    Alpha=1
    ConstraintWeights_0=0.7681
    ConstraintWeights_1:float=0.0283
    ConstraintWeights_2:float=0.0184
    ConstraintWeights_3:float=0.0062
    ConstraintWeights_4:float=0.1184
    ConstraintWeights_5:float=0.0114
    ConstraintWeights_6:float=0.0114
    ConstraintWeights_7:float=0.0178
    ConstraintWeights_8:float=0.0178
    ConstraintWeights_9:float=0.0012
    ConstraintWeights_10:float=0.0012
   ) -> (Pose) {
    alphaBoolBlend {blendOption=Linear}
    boneToModify {boneName=pouch1_l}
    constraintSetup [
     {targetBone {boneName=spine_01}}
     {targetBone {boneName=spine_02}}
     {targetBone {boneName=spine_03}}
     {targetBone {boneName=spine_04}}
     {targetBone {boneName=pelvis}}
     {targetBone {boneName=thigh_twist_01_l}}
     {targetBone {boneName=thigh_twist_01_r}}
     {targetBone {boneName=deform_thigh_fr_r}}
     {targetBone {boneName=deform_thigh_fr_l}}
     {targetBone {boneName=deform_thigh_bk_r}}
     {targetBone {boneName=deform_thigh_bk_l}}
    ]
    constraintWeights [1,1,1,1,1,1,1,1,1,1,1]
   }
   n11 = "Constraint - pouch2_l (PPPPPPPPPPP)"!(
    ComponentPose=@n10
    Alpha=1
    ConstraintWeights_0=0.8118
    ConstraintWeights_1:float=0.026
    ConstraintWeights_2:float=0.0181
    ConstraintWeights_3:float=0.0056
    ConstraintWeights_4:float=0.0817
    ConstraintWeights_5:float=0.0099
    ConstraintWeights_6:float=0.0099
    ConstraintWeights_7:float=0.0164
    ConstraintWeights_8:float=0.0164
    ConstraintWeights_9:float=0.0021
    ConstraintWeights_10:float=0.0021
   ) -> (Pose) {
    alphaBoolBlend {blendOption=Linear}
    boneToModify {boneName=pouch2_l}
    constraintSetup [
     {targetBone {boneName=spine_01}}
     {targetBone {boneName=spine_02}}
     {targetBone {boneName=spine_03}}
     {targetBone {boneName=spine_04}}
     {targetBone {boneName=pelvis}}
     {targetBone {boneName=thigh_twist_01_l}}
     {targetBone {boneName=thigh_twist_01_r}}
     {targetBone {boneName=deform_thigh_fr_r}}
     {targetBone {boneName=deform_thigh_fr_l}}
     {targetBone {boneName=deform_thigh_bk_r}}
     {targetBone {boneName=deform_thigh_bk_l}}
    ]
    constraintWeights [1,1,1,1,1,1,1,1,1,1,1]
   }
   n12 = "Constraint - pouch3_l (PPPPPPPPPPP)"!(
    ComponentPose=@n11
    Alpha=1
    ConstraintWeights_0=0.7976
    ConstraintWeights_1:float=0.0343
    ConstraintWeights_2:float=0.0262
    ConstraintWeights_3:float=0.0082
    ConstraintWeights_4:float=0.1016
    ConstraintWeights_5:float=0.0067
    ConstraintWeights_6:float=0.0067
    ConstraintWeights_7:float=0.0061
    ConstraintWeights_8:float=0.0061
    ConstraintWeights_9:float=0.0032
    ConstraintWeights_10:float=0.0032
   ) -> (Pose) {
    alphaBoolBlend {blendOption=Linear}
    boneToModify {boneName=pouch3_l}
    constraintSetup [
     {targetBone {boneName=spine_01}}
     {targetBone {boneName=spine_02}}
     {targetBone {boneName=spine_03}}
     {targetBone {boneName=spine_04}}
     {targetBone {boneName=pelvis}}
     {targetBone {boneName=thigh_twist_01_l}}
     {targetBone {boneName=thigh_twist_01_r}}
     {targetBone {boneName=deform_thigh_fr_r}}
     {targetBone {boneName=deform_thigh_fr_l}}
     {targetBone {boneName=deform_thigh_bk_r}}
     {targetBone {boneName=deform_thigh_bk_l}}
    ]
    constraintWeights [1,1,1,1,1,1,1,1,1,1,1]
   }
   n13 = "Constraint - pouch1_r (PPPPPPPPPPP)"!(
    ComponentPose=@n12
    Alpha=1
    ConstraintWeights_0=0.7681
    ConstraintWeights_1:float=0.0283
    ConstraintWeights_2:float=0.0184
    ConstraintWeights_3:float=0.0062
    ConstraintWeights_4:float=0.1184
    ConstraintWeights_5:float=0.0114
    ConstraintWeights_6:float=0.0114
    ConstraintWeights_7:float=0.0178
    ConstraintWeights_8:float=0.0178
    ConstraintWeights_9:float=0.0012
    ConstraintWeights_10:float=0.0012
   ) -> (Pose) {
    alphaBoolBlend {blendOption=Linear}
    boneToModify {boneName=pouch1_r}
    constraintSetup [
     {targetBone {boneName=spine_01}}
     {targetBone {boneName=spine_02}}
     {targetBone {boneName=spine_03}}
     {targetBone {boneName=spine_04}}
     {targetBone {boneName=pelvis}}
     {targetBone {boneName=thigh_twist_01_l}}
     {targetBone {boneName=thigh_twist_01_r}}
     {targetBone {boneName=deform_thigh_fr_r}}
     {targetBone {boneName=deform_thigh_fr_l}}
     {targetBone {boneName=deform_thigh_bk_r}}
     {targetBone {boneName=deform_thigh_bk_l}}
    ]
    constraintWeights [1,1,1,1,1,1,1,1,1,1,1]
   }
   n14 = "Constraint - pouch2_r (PPPPPPPPPPP)"!(
    ComponentPose=@n13
    Alpha=1
    ConstraintWeights_0=0.8118
    ConstraintWeights_1:float=0.026
    ConstraintWeights_2:float=0.0181
    ConstraintWeights_3:float=0.0056
    ConstraintWeights_4:float=0.0817
    ConstraintWeights_5:float=0.0099
    ConstraintWeights_6:float=0.0099
    ConstraintWeights_7:float=0.0164
    ConstraintWeights_8:float=0.0164
    ConstraintWeights_9:float=0.0021
    ConstraintWeights_10:float=0.0021
   ) -> (Pose) {
    alphaBoolBlend {blendOption=Linear}
    boneToModify {boneName=pouch2_r}
    constraintSetup [
     {targetBone {boneName=spine_01}}
     {targetBone {boneName=spine_02}}
     {targetBone {boneName=spine_03}}
     {targetBone {boneName=spine_04}}
     {targetBone {boneName=pelvis}}
     {targetBone {boneName=thigh_twist_01_l}}
     {targetBone {boneName=thigh_twist_01_r}}
     {targetBone {boneName=deform_thigh_fr_r}}
     {targetBone {boneName=deform_thigh_fr_l}}
     {targetBone {boneName=deform_thigh_bk_r}}
     {targetBone {boneName=deform_thigh_bk_l}}
    ]
    constraintWeights [1,1,1,1,1,1,1,1,1,1,1]
   }
   n15 = "Constraint - pouch3_r (PPPPPPPPPPP)"!(
    ComponentPose=@n14
    Alpha=1
    ConstraintWeights_0=0.7858
    ConstraintWeights_1:float=0.0347
    ConstraintWeights_2:float=0.0267
    ConstraintWeights_3:float=0.0082
    ConstraintWeights_4:float=0.1151
    ConstraintWeights_5:float=0.0063
    ConstraintWeights_6:float=0.0063
    ConstraintWeights_7:float=0.0048
    ConstraintWeights_8:float=0.0048
    ConstraintWeights_9:float=0.0036
    ConstraintWeights_10:float=0.0036
   ) -> (Pose) {
    alphaBoolBlend {blendOption=Linear}
    boneToModify {boneName=pouch3_r}
    constraintSetup [
     {targetBone {boneName=spine_01}}
     {targetBone {boneName=spine_02}}
     {targetBone {boneName=spine_03}}
     {targetBone {boneName=spine_04}}
     {targetBone {boneName=pelvis}}
     {targetBone {boneName=thigh_twist_01_l}}
     {targetBone {boneName=thigh_twist_01_r}}
     {targetBone {boneName=deform_thigh_fr_r}}
     {targetBone {boneName=deform_thigh_fr_l}}
     {targetBone {boneName=deform_thigh_bk_r}}
     {targetBone {boneName=deform_thigh_bk_l}}
    ]
    constraintWeights [1,1,1,1,1,1,1,1,1,1,1]
   }
   n16 = "Constraint - clavicle_cape_l (PPPPPPPPP)"!(
    ComponentPose=@n15
    Alpha=1
    ConstraintWeights_0=0.0042
    ConstraintWeights_1:float=0.0307
    ConstraintWeights_2:float=0.1816
    ConstraintWeights_3:float=0.6462
    ConstraintWeights_4:float=0.0065
    ConstraintWeights_5:float=0.1051
    ConstraintWeights_6:float=0.0138
    ConstraintWeights_7:float=0.0119
    ConstraintWeights_8:float=0
   ) -> (Pose) {
    alphaBoolBlend {blendOption=Linear}
    boneToModify {boneName=clavicle_cape_l}
    constraintSetup [
     {targetBone {boneName=spine_02}}
     {targetBone {boneName=spine_03}}
     {targetBone {boneName=spine_04}}
     {targetBone {boneName=spine_05}}
     {targetBone {boneName=clavicle_l}}
     {targetBone {boneName=clavicle_r}}
     {targetBone {boneName=neck_01}}
     {targetBone {boneName=neck_02}}
     {targetBone {boneName=upperarm_twist_01_r}}
    ]
    constraintWeights [1,1,1,1,1,1,1,1,1]
   }
   n17 = "Constraint - clavicle_cape_l (PPPPPPPPP)"(
    ComponentPose=@n16
    Alpha=1
    ConstraintWeights_0=0.0042
    ConstraintWeights_1=0.0307
    ConstraintWeights_2=0.1816
    ConstraintWeights_3=0.6462
    ConstraintWeights_4=0.1051
    ConstraintWeights_5=0.0065
    ConstraintWeights_6=0.0138
    ConstraintWeights_7=0.0119
    ConstraintWeights_8=0
   ) {
    alphaBoolBlend {blendOption=Linear}
    boneToModify {boneName=clavicle_cape_l}
    constraintSetup [
     {targetBone {boneName=spine_02}}
     {targetBone {boneName=spine_03}}
     {targetBone {boneName=spine_04}}
     {targetBone {boneName=spine_05}}
     {targetBone {boneName=clavicle_l}}
     {targetBone {boneName=clavicle_r}}
     {targetBone {boneName=neck_01}}
     {targetBone {boneName=neck_02}}
     {targetBone {boneName=upperarm_twist_01_l}}
    ]
    constraintWeights [1,1,1,1,1,1,1,1,1]
   }
   n18 = "Constraint - Strap2 (PPPPPPPP)"!(
    ComponentPose=@n17
    Alpha=1
    ConstraintWeights_0=0.0311
    ConstraintWeights_1:float=0.0033
    ConstraintWeights_2:float=0.0013
    ConstraintWeights_3:float=0.2035
    ConstraintWeights_4:float=0.0015
    ConstraintWeights_5:float=0.3973
    ConstraintWeights_6:float=0.3147
    ConstraintWeights_7:float=0.0474
   ) -> (Pose) {
    alphaBoolBlend {blendOption=Linear}
    boneToModify {boneName=Strap2}
    constraintSetup [
     {targetBone {boneName=spine_01}}
     {targetBone {boneName=spine_02}}
     {targetBone {boneName=spine_03}}
     {targetBone {boneName=pelvis}}
     {targetBone {boneName=thigh_twist_01_l}}
     {targetBone {boneName=thigh_twist_01_r}}
     {targetBone {boneName=deform_thigh_fr_r}}
     {targetBone {boneName=deform_thigh_bk_r}}
    ]
    constraintWeights [1,1,1,1,1,1,1,1]
   }
   n19 = "Transform (Modify) Bone - Bone: StrapPOSING"(ComponentPose=@n18, Alpha=Get legSide() -> (legSide:double), Translation="0,0,-1", Rotation="-10,0,-10", Scale="1,1,1") {
    rotationMode=BMM_Additive
    rotationSpace=BCS_BoneSpace
    scaleMode=BMM_Replace
    scaleSpace=BCS_BoneSpace
    translationMode=BMM_Additive
    translationSpace=BCS_BoneSpace
    alphaBoolBlend {blendOption=Linear}
    boneToModify {boneName=StrapPOSING}
    rotation {pitch=-10, roll=-10}
    translation {z=-1}
   }
   n20 = Get Is Crouching() -> (Is Crouching:double)
   n21 = "Transform (Modify) Bone - Bone: pouch2_l"(ComponentPose=@n19, Alpha=@n20, Translation="0,-3,1", Rotation="0,-30,0", Scale="1,1,1") {
    rotationMode=BMM_Additive
    rotationSpace=BCS_BoneSpace
    scaleMode=BMM_Replace
    scaleSpace=BCS_BoneSpace
    translationMode=BMM_Additive
    translationSpace=BCS_BoneSpace
    alphaBoolBlend {blendOption=Linear}
    boneToModify {boneName=pouch2_l}
    rotation {yaw=-30}
    translation {y=-3, z=1}
   }
   n22 = "Transform (Modify) Bone - Bone: pouch1_l"(ComponentPose=@n21, Alpha=@n20, Translation="-4,7,0", Rotation="0,30,0", Scale="1,1,1") {
    rotationMode=BMM_Additive
    rotationSpace=BCS_BoneSpace
    scaleMode=BMM_Replace
    scaleSpace=BCS_BoneSpace
    translationMode=BMM_Additive
    translationSpace=BCS_BoneSpace
    alphaBoolBlend {blendOption=Linear}
    boneToModify {boneName=pouch1_l}
    rotation {yaw=30}
    translation=<-4 7>
   }
   n23 = "Transform (Modify) Bone - Bone: pouch2_r"(ComponentPose=@n22, Alpha=@n20, Translation="0.5,4,0", Rotation="0,-20,0", Scale="1,1,1") {
    rotationMode=BMM_Additive
    rotationSpace=BCS_BoneSpace
    scaleMode=BMM_Replace
    scaleSpace=BCS_BoneSpace
    translationMode=BMM_Additive
    translationSpace=BCS_BoneSpace
    alphaBoolBlend {blendOption=Linear}
    boneToModify {boneName=pouch2_r}
    rotation {yaw=-20}
    translation=<0.5 4>
   }
   n24 = "Transform (Modify) Bone - Bone: pouch1_r"(ComponentPose=@n23, Alpha=@n20, Translation="1.5,-4,0", Rotation="0,20,0", Scale="1,1,1") {
    rotationMode=BMM_Additive
    rotationSpace=BCS_BoneSpace
    scaleMode=BMM_Replace
    scaleSpace=BCS_BoneSpace
    translationMode=BMM_Additive
    translationSpace=BCS_BoneSpace
    alphaBoolBlend {blendOption=Linear}
    boneToModify {boneName=pouch1_r}
    rotation {yaw=20}
    translation=<1.5 -4>
   }
   n25 = "Transform (Modify) Bone - Bone: spine_05_cape"(ComponentPose=@n24, Alpha=1, Translation="0,0,0", Rotation="0,0,0", Scale="1,1.1,1") {
    rotationMode=BMM_Additive
    rotationSpace=BCS_BoneSpace
    scaleMode=BMM_Replace
    scaleSpace=BCS_BoneSpace
    translationMode=BMM_Additive
    translationSpace=BCS_BoneSpace
    alphaBoolBlend {blendOption=Linear}
    boneToModify {boneName=spine_05_cape}
    scale {y=1.1}
   }
   n26 = Get isSkyDivingNoBB() -> (isSkyDivingNoBB:double)
   n27 = "Transform (Modify) Bone - Bone: spine_05_cape"(ComponentPose=@n25, Alpha=@n26, Translation="0,0,0", Rotation="0,0,0", Scale="1,1.2,1") {
    rotationMode=BMM_Additive
    rotationSpace=BCS_BoneSpace
    scaleMode=BMM_Replace
    scaleSpace=BCS_BoneSpace
    translationMode=BMM_Additive
    translationSpace=BCS_BoneSpace
    alphaBoolBlend {blendOption=Linear}
    boneToModify {boneName=spine_05_cape}
    scale {y=1.2}
   }
   n28 = "Transform (Modify) Bone - Bone: clavicle_l_chest_cape"(ComponentPose=@n27, Alpha=@n26, Translation="0,0,0", Rotation="0,0,0", Scale="1,2,1") {
    rotationMode=BMM_Additive
    rotationSpace=BCS_BoneSpace
    scaleMode=BMM_Replace
    scaleSpace=BCS_BoneSpace
    translationMode=BMM_Additive
    translationSpace=BCS_BoneSpace
    alphaBoolBlend {blendOption=Linear}
    boneToModify {boneName=clavicle_l_chest_cape}
    scale {y=2}
   }
   n29 = "Transform (Modify) Bone - Bone: clavicle_l_chest_cape"(ComponentPose=@n28, Alpha=@n26, Translation="0,0,0", Rotation="0,0,0", Scale="1,1.25,1") {
    rotationMode=BMM_Additive
    rotationSpace=BCS_BoneSpace
    scaleMode=BMM_Replace
    scaleSpace=BCS_BoneSpace
    translationMode=BMM_Additive
    translationSpace=BCS_BoneSpace
    alphaBoolBlend {blendOption=Linear}
    boneToModify {boneName=clavicle_l_chest_cape}
    scale {y=1.25}
   }
   n30 = "Transform (Modify) Bone - Bone: spine_05_cape"(ComponentPose=@n29, Alpha=Get ClavFwdLf() -> (ClavFwdLf:double), Translation="0,-5,0", Rotation="0,0,0", Scale="1,1,1") {
    rotationMode=BMM_Additive
    rotationSpace=BCS_BoneSpace
    scaleMode=BMM_Replace
    scaleSpace=BCS_BoneSpace
    translationMode=BMM_Additive
    translationSpace=BCS_BoneSpace
    alphaBoolBlend {blendOption=Linear}
    boneToModify {boneName=spine_05_cape}
    translation {y=-5}
   }
   n31 = Get Clavicle LRot() -> (Clavicle LRot:double)
   n32 = "Transform (Modify) Bone - Bone: clavicle_l_cape"(ComponentPose=@n30, Alpha=@n31, Translation="40,-40,60", Rotation="0,0,0", Scale="1,1,1") {
    rotationMode=BMM_Additive
    rotationSpace=BCS_BoneSpace
    scaleMode=BMM_Replace
    scaleSpace=BCS_BoneSpace
    translationMode=BMM_Additive
    translationSpace=BCS_BoneSpace
    alphaBoolBlend {blendOption=Linear}
    boneToModify {boneName=clavicle_l_cape}
    translation=<40 -40 60>
   }
   n33 = "Transform (Modify) Bone - Bone: deform_trap_clav_l"(ComponentPose=@n32, Alpha=@n31, Translation="0,0,20", Rotation="0,0,0", Scale="1,1,1") {
    rotationMode=BMM_Additive
    rotationSpace=BCS_BoneSpace
    scaleMode=BMM_Replace
    scaleSpace=BCS_BoneSpace
    translationMode=BMM_Additive
    translationSpace=BCS_BoneSpace
    alphaBoolBlend {blendOption=Linear}
    boneToModify {boneName=deform_trap_clav_l}
    translation {z=20}
   }
   n34 = "Transform (Modify) Bone - Bone: clavicle_l_chest_cape"(ComponentPose=@n33, Alpha=@n31, Translation="0,0,0", Rotation="0,0,0", Scale="0,2,0") {
    rotationMode=BMM_Additive
    rotationSpace=BCS_BoneSpace
    scaleMode=BMM_Replace
    scaleSpace=BCS_BoneSpace
    translationMode=BMM_Additive
    translationSpace=BCS_BoneSpace
    alphaBoolBlend {blendOption=Linear}
    boneToModify {boneName=clavicle_l_chest_cape}
    scale=<0 2 0>
   }
   n35 = "Constraint - clavicle_cape_l (RR)"!(ComponentPose=@n34, Alpha=1, ConstraintWeights_0=0, ConstraintWeights_1:float=1) -> (Pose) {
    alphaBoolBlend {blendOption=Linear}
    boneToModify {boneName=clavicle_cape_l}
    constraintSetup [{transformType=Rotation, targetBone {boneName=clavicle_l}}, {transformType=Rotation, targetBone {boneName=spine_05}}]
    constraintWeights [1,1]
   }
   n36 = "Constraint - clavicle_cape_r (RR)"!(ComponentPose=@n35, Alpha=1, ConstraintWeights_0=0, ConstraintWeights_1:float=1) -> (Pose) {
    alphaBoolBlend {blendOption=Linear}
    boneToModify {boneName=clavicle_cape_r}
    constraintSetup [{transformType=Rotation, targetBone {boneName=clavicle_r}}, {transformType=Rotation, targetBone {boneName=spine_05}}]
    constraintWeights [1,1]
   }
   n37 = Get RigidBodyOffSkydiving() -> (RigidBodyOffSkydiving:double)
   n38 = "Transform (Modify) Bone - Bone: cape_00_r"(ComponentPose=@n36, Alpha=@n37, Translation="0,0,0", Rotation="0,-15,-15", Scale="1,1,1") {
    rotationMode=BMM_Additive
    rotationSpace=BCS_BoneSpace
    scaleMode=BMM_Replace
    scaleSpace=BCS_BoneSpace
    translationMode=BMM_Additive
    translationSpace=BCS_BoneSpace
    alphaBoolBlend {blendOption=Linear}
    boneToModify {boneName=cape_00_r}
    rotation {roll=-15, yaw=-15}
   }
   n39 = "Transform (Modify) Bone - Bone: cape_mid_01"(ComponentPose=@n38, Alpha=@n37, Translation="0,0,0", Rotation="0,-20,0", Scale="1,1,1") {
    rotationMode=BMM_Additive
    rotationSpace=BCS_BoneSpace
    scaleMode=BMM_Replace
    scaleSpace=BCS_BoneSpace
    translationMode=BMM_Additive
    translationSpace=BCS_BoneSpace
    alphaBoolBlend {blendOption=Linear}
    boneToModify {boneName=cape_mid_01}
    rotation {yaw=-20}
   }
   n40 = "Transform (Modify) Bone - Bone: cape_00_l"(ComponentPose=@n39, Alpha=@n37, Translation="0,0,0", Rotation="0,-15,-15", Scale="1,1,1") {
    rotationMode=BMM_Additive
    rotationSpace=BCS_BoneSpace
    scaleMode=BMM_Replace
    scaleSpace=BCS_BoneSpace
    translationMode=BMM_Additive
    translationSpace=BCS_BoneSpace
    alphaBoolBlend {blendOption=Linear}
    boneToModify {boneName=cape_00_l}
    rotation {roll=-15, yaw=-15}
   }
   n41 = Get GravityOverride()
   n42 = RigidBody!(ComponentPose=@n40, Alpha=Get NoBBNoFE() -> (NoBBNoFE:double), OverrideWorldGravity:Vector=@n41, ExternalForce) -> (Pose) {
    bClampLinearTranslationLimitToRefPose=true
    bOverrideWorldGravity=true
    overridePhysicsAsset=&/M_MED_Sienna_DynCapeMain
    simulationSpace=BaseBoneSpace
    alphaBoolBlend {blendOption=Linear}
    baseBoneRef {boneName=spine_05}
    componentAppliedLinearAccClamp=<3000 3000 3000>
    componentLinearAccScale=<0.5 0.3 0.3>
    componentLinearVelScale=<0.5 0.3 0.3>
    overrideWorldGravity {z=-980}
    simSpaceSettings {
     maxAngularAcceleration=1000
     maxAngularVelocity=3
     maxLinearAcceleration=700
     maxLinearVelocity=500
     velocityScaleZ=0
     worldAlpha=1
     externalLinearDragV=<0.2 0.2>
    }
   }
   n43 = RigidBody(ComponentPose=@n42, Alpha=Get IsWearingBBNoFE() -> (IsWearingBBNoFE:double), OverrideWorldGravity=@n41) {
    bClampLinearTranslationLimitToRefPose=true
    bOverrideWorldGravity=true
    overridePhysicsAsset=&/M_MED_Sienna_DynCapeBBSkydive
    simulationSpace=BaseBoneSpace
    alphaBoolBlend {blendOption=Linear}
    baseBoneRef {boneName=spine_05}
    componentAppliedLinearAccClamp=<3000 3000 3000>
    componentLinearAccScale=<0.5 0.3 0.3>
    componentLinearVelScale=<0.5 0.3 0.3>
    overrideWorldGravity {z=-980}
    simSpaceSettings {
     maxAngularAcceleration=1000
     maxAngularVelocity=3
     maxLinearAcceleration=700
     maxLinearVelocity=500
     velocityScaleZ=0
     worldAlpha=1
     externalLinearDragV=<0.2 0.2>
    }
   }
   n44 = RigidBody!(ComponentPose=@n43, Alpha=Get FrontendNoEmote() -> (FrontendNoEmote:double), ExternalForce) -> (Pose) {
    bClampLinearTranslationLimitToRefPose=true
    bOverrideWorldGravity=true
    overridePhysicsAsset=&/M_MED_Sienna_DynCapeMainFE
    simulationSpace=BaseBoneSpace
    alphaBoolBlend {blendOption=Linear}
    baseBoneRef {boneName=spine_05}
    componentAppliedLinearAccClamp=<3000 3000 3000>
    componentLinearAccScale=<0.4 0.3 0.3>
    componentLinearVelScale=<0.4 0.3 0.3>
    overrideWorldGravity {z=-980}
   }
   n45 = RigidBody(ComponentPose=@n44, Alpha=Get isPlayingEmoteFE() -> (isPlayingEmoteFE:double)) {
    bClampLinearTranslationLimitToRefPose=true
    bOverrideWorldGravity=true
    overridePhysicsAsset=&M_MED_Sienna_DynCapeMain
    simulationSpace=BaseBoneSpace
    alphaBoolBlend {blendOption=Linear}
    baseBoneRef {boneName=spine_05}
    componentAppliedLinearAccClamp=<3000 3000 3000>
    componentLinearAccScale=<0.4 0.3 0.3>
    componentLinearVelScale=<0.4 0.3 0.3>
    overrideWorldGravity {z=-980}
   }
   n46 = Get wearing backpack and skydiving() -> (wearing backpack and skydiving:double)
   n47 = "Transform (Modify) Bone - Bone: cape_00_r"(ComponentPose=@n45, Alpha=@n46, Translation="0,0,0", Rotation="0,4,7", Scale="1,1,1") {
    rotationMode=BMM_Additive
    rotationSpace=BCS_BoneSpace
    scaleMode=BMM_Replace
    scaleSpace=BCS_BoneSpace
    translationMode=BMM_Additive
    translationSpace=BCS_BoneSpace
    alphaBoolBlend {blendOption=Linear}
    boneToModify {boneName=cape_00_r}
    rotation {roll=7, yaw=4}
   }
   n48 = "Transform (Modify) Bone - Bone: cape_mid_01"(ComponentPose=@n47, Alpha=@n46, Translation="0,0,0", Rotation="0,5,0", Scale="1,1,1") {
    rotationMode=BMM_Additive
    rotationSpace=BCS_BoneSpace
    scaleMode=BMM_Replace
    scaleSpace=BCS_BoneSpace
    translationMode=BMM_Additive
    translationSpace=BCS_BoneSpace
    alphaBoolBlend {blendOption=Linear}
    boneToModify {boneName=cape_mid_01}
    rotation {yaw=5}
   }
   n49 = "Transform (Modify) Bone - Bone: cape_00_l"(ComponentPose=@n48, Alpha=@n46, Translation="0,0,0", Rotation="0,4,4", Scale="1,1,1") {
    rotationMode=BMM_Additive
    rotationSpace=BCS_BoneSpace
    scaleMode=BMM_Replace
    scaleSpace=BCS_BoneSpace
    translationMode=BMM_Additive
    translationSpace=BCS_BoneSpace
    alphaBoolBlend {blendOption=Linear}
    boneToModify {boneName=cape_00_l}
    rotation {roll=4, yaw=4}
   }
   n50 = Get isParachuteOpenedNoBB() -> (isParachuteOpenedNoBB:double)
   n51 = "Transform (Modify) Bone - Bone: cape_00_r"(ComponentPose=@n49, Alpha=@n50, Translation="0,0,0", Rotation="0,-8,-15", Scale="1,1,1") {
    rotationMode=BMM_Additive
    rotationSpace=BCS_BoneSpace
    scaleMode=BMM_Replace
    scaleSpace=BCS_BoneSpace
    translationMode=BMM_Additive
    translationSpace=BCS_BoneSpace
    alphaBoolBlend {blendOption=Linear}
    boneToModify {boneName=cape_00_r}
    rotation {roll=-15, yaw=-8}
   }
   n52 = "Transform (Modify) Bone - Bone: cape_mid_01"(ComponentPose=@n51, Alpha=@n50, Translation="0,0,0", Rotation="0,-10,0", Scale="1,1,1") {
    rotationMode=BMM_Additive
    rotationSpace=BCS_BoneSpace
    scaleMode=BMM_Replace
    scaleSpace=BCS_BoneSpace
    translationMode=BMM_Additive
    translationSpace=BCS_BoneSpace
    alphaBoolBlend {blendOption=Linear}
    boneToModify {boneName=cape_mid_01}
    rotation {yaw=-10}
   }
   n53 = "Transform (Modify) Bone - Bone: cape_00_l"(ComponentPose=@n52, Alpha=@n50, Translation="0,0,0", Rotation="0,-8,-8", Scale="1,1,1") {
    rotationMode=BMM_Additive
    rotationSpace=BCS_BoneSpace
    scaleMode=BMM_Replace
    scaleSpace=BCS_BoneSpace
    translationMode=BMM_Additive
    translationSpace=BCS_BoneSpace
    alphaBoolBlend {blendOption=Linear}
    boneToModify {boneName=cape_00_l}
    rotation {roll=-8, yaw=-8}
   }
   n54 = "Transform (Modify) Bone - Bone: dyn_collar_bk"(ComponentPose=@n53, Alpha=Get HeadRotBk() -> (HeadRotBk:double), Translation="0,4,4", Rotation="0,0,0", Scale="1,1,1") {
    rotationMode=BMM_Additive
    rotationSpace=BCS_BoneSpace
    scaleMode=BMM_Replace
    scaleSpace=BCS_BoneSpace
    translationMode=BMM_Additive
    translationSpace=BCS_BoneSpace
    alphaBoolBlend {blendOption=Linear}
    boneToModify {boneName=dyn_collar_bk}
    translation {y=4, z=4}
   }
   n55 = "Transform (Modify) Bone - Bone: dyn_collar_fr_MSTR"(ComponentPose=@n54, Alpha=Get HeadRotFr() -> (HeadRotFr:double), Translation="-2,-2,0", Rotation="15,0,5", Scale="1,1,1") {
    rotationMode=BMM_Additive
    rotationSpace=BCS_BoneSpace
    scaleMode=BMM_Replace
    scaleSpace=BCS_BoneSpace
    translationMode=BMM_Additive
    translationSpace=BCS_BoneSpace
    alphaBoolBlend {blendOption=Linear}
    boneToModify {boneName=dyn_collar_fr_MSTR}
    rotation {pitch=15, roll=5}
    translation=<-2 -2>
   }
   n56 = Get FrontendNoEmote()
   n57 = "Transform (Modify) Bone - Bone: cape_00_r"(ComponentPose=@n55, Alpha=@n56, Translation="-4,4,4", Rotation="5,5,0", Scale="1,1,1") {
    rotationMode=BMM_Additive
    rotationSpace=BCS_BoneSpace
    scaleMode=BMM_Replace
    scaleSpace=BCS_BoneSpace
    translationMode=BMM_Additive
    translationSpace=BCS_BoneSpace
    alphaBoolBlend {blendOption=Linear}
    boneToModify {boneName=cape_00_r}
    rotation {pitch=5, yaw=5}
    translation=<-4 4 4>
   }
   n58 = "Transform (Modify) Bone - Bone: cape_02_r"(ComponentPose=@n57, Alpha=@n56, Translation="-2.735,-4.835,-4.526", Rotation="0,0,10", Scale="1,1,1") {
    rotationMode=BMM_Additive
    rotationSpace=BCS_BoneSpace
    scaleMode=BMM_Replace
    scaleSpace=BCS_BoneSpace
    translationMode=BMM_Additive
    translationSpace=BCS_BoneSpace
    alphaBoolBlend {blendOption=Linear}
    boneToModify {boneName=cape_02_r}
    rotation {roll=10}
    translation=<-2.735 -4.8349 -4.5256>
   }
   n59 = "Transform (Modify) Bone - Bone: clavicle_l_frontendStrap"(ComponentPose=@n58, Alpha=@n56, Translation="-10,0,-7", Rotation="55,0,0", Scale="1,1,1") {
    rotationMode=BMM_Additive
    rotationSpace=BCS_BoneSpace
    scaleMode=BMM_Replace
    scaleSpace=BCS_BoneSpace
    translationMode=BMM_Additive
    translationSpace=BCS_BoneSpace
    alphaBoolBlend {blendOption=Linear}
    boneToModify {boneName=clavicle_l_frontendStrap}
    rotation {pitch=55}
    translation {x=-10, z=-7}
   }
   n60 = "Transform (Modify) Bone - Bone: dyn_backpack_socket"!(
    ComponentPose=@n59
    bAlphaBoolEnabled:bool=Get cape_active() -> (cape_active:bool)
    Translation="0.000,-5,0.000"
    Rotation="0,0,0"
    Scale="1,1,1"
   ) -> (Pose) {
    alphaInputType=Bool
    rotationMode=BMM_Additive
    rotationSpace=BCS_BoneSpace
    translationMode=BMM_Additive
    translationSpace=BCS_BoneSpace
    alphaBoolBlend {blendOption=Linear}
    boneToModify {boneName=dyn_backpack_socket}
    translation {y=-5}
   }
   Save cached pose 'previous setup'(Pose=Component To Local(ComponentPose=@n60)) {cacheName=previous setup}
   n61 = Anim Dynamics - Bone: cape_mid_01(ComponentPose=Local To Component(LocalPose=Use cached pose 'previous setup'()), Alpha=1, ExternalForce="0,0,0") {
    angularDampingOverride=1
    angularSpringConstant=10000
    bAngularSpring=true
    bOverrideAngularDamping=true
    bUseGravityOverride=true
    gravityScale=0.35
    simulationSpace=BoneRelative
    alphaBoolBlend {blendOption=Linear}
    boundBone {boneName=cape_mid_01}
    componentAppliedLinearAccClamp=<3000 3000 3000>
    componentLinearAccScale=<2 2 2>
    componentLinearVelScale=<2 2 2>
    gravityOverride {z=-980}
    physicsBodyDefinitions [{
      boundBone {boneName=cape_mid_01}
      boxExtents=<10 40 40>
      constraintSetup {angularLimitsMax=<0 5 0>, angularLimitsMin=<0 -20 -15>, angularTarget=<1 0 0>}
      localJointOffset=<5 0 0>
     }]
    relativeSpaceBone {boneName=backpack_root}
   }
   n62 = Anim Dynamics - Bone: cape_mid_03(ComponentPose=@n61, Alpha=1, ExternalForce="0,0,0") {
    angularDampingOverride=1
    angularSpringConstant=10000
    bAngularSpring=true
    bOverrideAngularDamping=true
    bUseGravityOverride=true
    gravityScale=0.35
    simulationSpace=BoneRelative
    alphaBoolBlend {blendOption=Linear}
    boundBone {boneName=cape_mid_03}
    componentAppliedLinearAccClamp=<3000 3000 3000>
    componentLinearAccScale=<2 2 2>
    componentLinearVelScale=<2 2 2>
    gravityOverride {z=-980}
    physicsBodyDefinitions [{
      boundBone {boneName=cape_mid_03}
      boxExtents=<10 40 40>
      constraintSetup {angularLimitsMax=<0 5 0>, angularLimitsMin=<0 -20 -15>, angularTarget=<1 0 0>}
      localJointOffset=<5 0 0>
     }]
    relativeSpaceBone {boneName=backpack_root}
   }
   n63 = Anim Dynamics - Bone: cape_mid_05(ComponentPose=@n62, Alpha=1, ExternalForce="0,0,0") {
    angularDampingOverride=1
    angularSpringConstant=10000
    bAngularSpring=true
    bOverrideAngularDamping=true
    bUseGravityOverride=true
    gravityScale=0.35
    simulationSpace=BoneRelative
    alphaBoolBlend {blendOption=Linear}
    boundBone {boneName=cape_mid_05}
    componentAppliedLinearAccClamp=<3000 3000 3000>
    componentLinearAccScale=<2 2 2>
    componentLinearVelScale=<2 2 2>
    gravityOverride {z=-980}
    physicsBodyDefinitions [{
      boundBone {boneName=cape_mid_05}
      boxExtents=<10 40 40>
      constraintSetup {angularLimitsMax=<0 5 0>, angularLimitsMin=<0 -20 -15>, angularTarget=<1 0 0>}
      localJointOffset=<5 0 0>
     }]
    relativeSpaceBone {boneName=backpack_root}
   }
   n64 = Anim Dynamics - Bone: cape_mid_06(ComponentPose=@n63, Alpha=1, ExternalForce="0,0,0") {
    angularDampingOverride=1
    angularSpringConstant=10000
    bAngularSpring=true
    bOverrideAngularDamping=true
    bUseGravityOverride=true
    gravityScale=0.35
    simulationSpace=BoneRelative
    alphaBoolBlend {blendOption=Linear}
    boundBone {boneName=cape_mid_06}
    componentAppliedLinearAccClamp=<3000 3000 3000>
    componentLinearAccScale=<2 2 2>
    componentLinearVelScale=<2 2 2>
    gravityOverride {z=-980}
    physicsBodyDefinitions [{
      boundBone {boneName=cape_mid_06}
      boxExtents=<10 40 40>
      constraintSetup {angularLimitsMax=<0 5 0>, angularLimitsMin=<0 -20 -15>, angularTarget=<1 0 0>}
      localJointOffset=<5 0 0>
     }]
    relativeSpaceBone {boneName=backpack_root}
   }
   n65 = Anim Dynamics - Bone: cape_01_r(ComponentPose=@n64, Alpha=1, ExternalForce="0,0,0") {
    angularDampingOverride=1
    angularSpringConstant=10000
    bAngularSpring=true
    bOverrideAngularDamping=true
    bUseGravityOverride=true
    gravityScale=0.35
    simulationSpace=BoneRelative
    alphaBoolBlend {blendOption=Linear}
    boundBone {boneName=cape_01_r}
    componentAppliedLinearAccClamp=<3000 3000 3000>
    componentLinearAccScale=<2 2 2>
    componentLinearVelScale=<2 2 2>
    gravityOverride {z=-980}
    physicsBodyDefinitions [{
      boundBone {boneName=cape_01_r}
      boxExtents=<10 40 40>
      constraintSetup {angularLimitsMax=<0 5 0>, angularLimitsMin=<0 -20 -15>, angularTarget=<1 0 0>}
      localJointOffset=<5 0 0>
     }]
    relativeSpaceBone {boneName=backpack_root}
   }
   n66 = Anim Dynamics - Bone: cape_03_r(ComponentPose=@n65, Alpha=1, ExternalForce="0,0,0") {
    angularDampingOverride=1
    angularSpringConstant=10000
    bAngularSpring=true
    bOverrideAngularDamping=true
    bUseGravityOverride=true
    gravityScale=0.35
    simulationSpace=BoneRelative
    alphaBoolBlend {blendOption=Linear}
    boundBone {boneName=cape_03_r}
    componentAppliedLinearAccClamp=<3000 3000 3000>
    componentLinearAccScale=<2 2 2>
    componentLinearVelScale=<2 2 2>
    gravityOverride {z=-980}
    physicsBodyDefinitions [{
      boundBone {boneName=cape_03_r}
      boxExtents=<10 40 40>
      constraintSetup {angularLimitsMax=<0 5 0>, angularLimitsMin=<0 -20 -15>, angularTarget=<1 0 0>}
      localJointOffset=<5 0 0>
     }]
    relativeSpaceBone {boneName=backpack_root}
   }
   n67 = Anim Dynamics - Bone: cape_05_r(ComponentPose=@n66, Alpha=1, ExternalForce="0,0,0") {
    angularDampingOverride=1
    angularSpringConstant=10000
    bAngularSpring=true
    bOverrideAngularDamping=true
    bUseGravityOverride=true
    gravityScale=0.35
    simulationSpace=BoneRelative
    alphaBoolBlend {blendOption=Linear}
    boundBone {boneName=cape_05_r}
    componentAppliedLinearAccClamp=<3000 3000 3000>
    componentLinearAccScale=<2 2 2>
    componentLinearVelScale=<2 2 2>
    gravityOverride {z=-980}
    physicsBodyDefinitions [{
      boundBone {boneName=cape_05_r}
      boxExtents=<10 40 40>
      constraintSetup {angularLimitsMax=<0 5 0>, angularLimitsMin=<0 -20 -15>, angularTarget=<1 0 0>}
      localJointOffset=<5 0 0>
     }]
    relativeSpaceBone {boneName=backpack_root}
   }
   n68 = Anim Dynamics - Bone: cape_07_r(ComponentPose=@n67, Alpha=1, ExternalForce="0,0,0") {
    angularDampingOverride=1
    angularSpringConstant=10000
    bAngularSpring=true
    bOverrideAngularDamping=true
    bUseGravityOverride=true
    gravityScale=0.35
    simulationSpace=BoneRelative
    alphaBoolBlend {blendOption=Linear}
    boundBone {boneName=cape_07_r}
    componentAppliedLinearAccClamp=<3000 3000 3000>
    componentLinearAccScale=<2 2 2>
    componentLinearVelScale=<2 2 2>
    gravityOverride {z=-980}
    physicsBodyDefinitions [{
      boundBone {boneName=cape_07_r}
      boxExtents=<10 40 40>
      constraintSetup {angularLimitsMax=<0 5 0>, angularLimitsMin=<0 -20 -15>, angularTarget=<1 0 0>}
      localJointOffset=<5 0 0>
     }]
    relativeSpaceBone {boneName=backpack_root}
   }
   n69 = Anim Dynamics - Bone: cape_03_l(ComponentPose=@n68, Alpha=1, ExternalForce="0,0,0") {
    angularDampingOverride=1
    angularSpringConstant=10000
    bAngularSpring=true
    bOverrideAngularDamping=true
    bUseGravityOverride=true
    gravityScale=0.35
    simulationSpace=BoneRelative
    alphaBoolBlend {blendOption=Linear}
    boundBone {boneName=cape_03_l}
    componentAppliedLinearAccClamp=<3000 3000 3000>
    componentLinearAccScale=<2 2 2>
    componentLinearVelScale=<2 2 2>
    gravityOverride {z=-980}
    physicsBodyDefinitions [{
      boundBone {boneName=cape_03_l}
      boxExtents=<10 40 40>
      constraintSetup {angularLimitsMax=<0 20 0>, angularLimitsMin=<0 -20 -15>, angularTarget=<1 0 0>}
      localJointOffset=<5 0 0>
     }]
    relativeSpaceBone {boneName=backpack_root}
   }
   n70 = Anim Dynamics - Bone: cape_05_l(ComponentPose=@n69, Alpha=1, ExternalForce="0,0,0") {
    angularDampingOverride=1
    angularSpringConstant=10000
    bAngularSpring=true
    bOverrideAngularDamping=true
    bUseGravityOverride=true
    gravityScale=0.35
    simulationSpace=BoneRelative
    alphaBoolBlend {blendOption=Linear}
    boundBone {boneName=cape_05_l}
    componentAppliedLinearAccClamp=<3000 3000 3000>
    componentLinearAccScale=<2 2 2>
    componentLinearVelScale=<2 2 2>
    gravityOverride {z=-980}
    physicsBodyDefinitions [{
      boundBone {boneName=cape_05_l}
      boxExtents=<10 40 40>
      constraintSetup {angularLimitsMax=<0 20 0>, angularLimitsMin=<0 -20 -15>, angularTarget=<1 0 0>}
      localJointOffset=<5 0 0>
     }]
    relativeSpaceBone {boneName=backpack_root}
   }
   n71 = Anim Dynamics - Bone: cape_07_l(ComponentPose=@n70, Alpha=1, ExternalForce="0,0,0") {
    angularDampingOverride=1
    angularSpringConstant=10000
    bAngularSpring=true
    bOverrideAngularDamping=true
    bUseGravityOverride=true
    gravityScale=0.35
    simulationSpace=BoneRelative
    alphaBoolBlend {blendOption=Linear}
    boundBone {boneName=cape_07_l}
    componentAppliedLinearAccClamp=<3000 3000 3000>
    componentLinearAccScale=<2 2 2>
    componentLinearVelScale=<2 2 2>
    gravityOverride {z=-980}
    physicsBodyDefinitions [{
      boundBone {boneName=cape_07_l}
      boxExtents=<10 40 40>
      constraintSetup {angularLimitsMax=<0 20 0>, angularLimitsMin=<0 -20 -15>, angularTarget=<1 0 0>}
      localJointOffset=<5 0 0>
     }]
    relativeSpaceBone {boneName=backpack_root}
   }
   n72 = Anim Dynamics - Bone: cape_mid_01(ComponentPose=Local To Component(LocalPose=Use cached pose 'previous setup'()), Alpha=1, ExternalForce="0,0,0") {
    angularDampingOverride=1
    angularSpringConstant=10000
    bAngularSpring=true
    bOverrideAngularDamping=true
    bUseGravityOverride=true
    gravityScale=0.35
    simulationSpace=World
    alphaBoolBlend {blendOption=Linear}
    boundBone {boneName=cape_mid_01}
    componentAppliedLinearAccClamp=<3000 3000 3000>
    componentLinearAccScale=<2 2 2>
    componentLinearVelScale=<2 2 2>
    gravityOverride {z=-980}
    physicsBodyDefinitions [{
      boundBone {boneName=cape_mid_01}
      boxExtents=<30 40 40>
      constraintSetup {angularLimitsMax=<0 5 5>, angularLimitsMin=<0 -5 -15>, angularTarget=<1 0 0>}
      localJointOffset=<4 0 0>
     }]
    relativeSpaceBone {boneName=backpack_root}
   }
   n73 = Anim Dynamics - Bone: cape_mid_03(ComponentPose=@n72, Alpha=1, ExternalForce="0,0,0") {
    angularDampingOverride=1
    angularSpringConstant=10000
    bAngularSpring=true
    bOverrideAngularDamping=true
    bUseGravityOverride=true
    gravityScale=0.35
    simulationSpace=World
    alphaBoolBlend {blendOption=Linear}
    boundBone {boneName=cape_mid_03}
    componentAppliedLinearAccClamp=<3000 3000 3000>
    componentLinearAccScale=<2 2 2>
    componentLinearVelScale=<2 2 2>
    gravityOverride {z=-980}
    physicsBodyDefinitions [{
      boundBone {boneName=cape_mid_03}
      boxExtents=<30 40 40>
      constraintSetup {angularLimitsMax=<0 5 15>, angularLimitsMin=<0 -5 -15>, angularTarget=<1 0 0>}
      localJointOffset=<4 0 0>
     }]
    relativeSpaceBone {boneName=backpack_root}
   }
   n74 = Anim Dynamics - Bone: cape_mid_05(ComponentPose=@n73, Alpha=1, ExternalForce="0,0,0") {
    angularDampingOverride=1
    angularSpringConstant=10000
    bAngularSpring=true
    bOverrideAngularDamping=true
    bUseGravityOverride=true
    gravityScale=0.35
    simulationSpace=World
    alphaBoolBlend {blendOption=Linear}
    boundBone {boneName=cape_mid_05}
    componentAppliedLinearAccClamp=<3000 3000 3000>
    componentLinearAccScale=<2 2 2>
    componentLinearVelScale=<2 2 2>
    gravityOverride {z=-980}
    physicsBodyDefinitions [{
      boundBone {boneName=cape_mid_05}
      boxExtents=<30 40 40>
      constraintSetup {angularLimitsMax=<0 5 15>, angularLimitsMin=<0 -5 -15>, angularTarget=<1 0 0>}
      localJointOffset=<4 0 0>
     }]
    relativeSpaceBone {boneName=backpack_root}
   }
   n75 = Anim Dynamics - Bone: cape_mid_06(ComponentPose=@n74, Alpha=1, ExternalForce="0,0,0") {
    angularDampingOverride=1
    angularSpringConstant=10000
    bAngularSpring=true
    bOverrideAngularDamping=true
    bUseGravityOverride=true
    gravityScale=0.35
    simulationSpace=World
    alphaBoolBlend {blendOption=Linear}
    boundBone {boneName=cape_mid_06}
    componentAppliedLinearAccClamp=<3000 3000 3000>
    componentLinearAccScale=<2 2 2>
    componentLinearVelScale=<2 2 2>
    gravityOverride {z=-980}
    physicsBodyDefinitions [{
      boundBone {boneName=cape_mid_06}
      boxExtents=<30 40 40>
      constraintSetup {angularLimitsMax=<0 5 15>, angularLimitsMin=<0 -5 -15>, angularTarget=<1 0 0>}
      localJointOffset=<4 0 0>
     }]
    relativeSpaceBone {boneName=backpack_root}
   }
   n76 = Anim Dynamics - Bone: cape_01_r(ComponentPose=@n75, Alpha=1, ExternalForce="0,0,0") {
    angularDampingOverride=1
    angularSpringConstant=10000
    bAngularSpring=true
    bOverrideAngularDamping=true
    bUseGravityOverride=true
    gravityScale=0.35
    simulationSpace=World
    alphaBoolBlend {blendOption=Linear}
    boundBone {boneName=cape_01_r}
    componentAppliedLinearAccClamp=<3000 3000 3000>
    componentLinearAccScale=<2 2 2>
    componentLinearVelScale=<2 2 2>
    gravityOverride {z=-980}
    physicsBodyDefinitions [{
      boundBone {boneName=cape_01_r}
      boxExtents=<30 40 40>
      constraintSetup {angularLimitsMax=<0 5 5>, angularLimitsMin=<0 -5 -15>, angularTarget=<1 0 0>}
      localJointOffset=<4 0 0>
     }]
    relativeSpaceBone {boneName=backpack_root}
   }
   n77 = Anim Dynamics - Bone: cape_03_r(ComponentPose=@n76, Alpha=1, ExternalForce="0,0,0") {
    angularDampingOverride=1
    angularSpringConstant=10000
    bAngularSpring=true
    bOverrideAngularDamping=true
    bUseGravityOverride=true
    gravityScale=0.35
    simulationSpace=World
    alphaBoolBlend {blendOption=Linear}
    boundBone {boneName=cape_03_r}
    componentAppliedLinearAccClamp=<3000 3000 3000>
    componentLinearAccScale=<2 2 2>
    componentLinearVelScale=<2 2 2>
    gravityOverride {z=-980}
    physicsBodyDefinitions [{
      boundBone {boneName=cape_03_r}
      boxExtents=<30 40 40>
      constraintSetup {angularLimitsMax=<0 5 15>, angularLimitsMin=<0 -5 -15>, angularTarget=<1 0 0>}
      localJointOffset=<4 0 0>
     }]
    relativeSpaceBone {boneName=backpack_root}
   }
   n78 = Anim Dynamics - Bone: cape_05_r(ComponentPose=@n77, Alpha=1, ExternalForce="0,0,0") {
    angularDampingOverride=1
    angularSpringConstant=10000
    bAngularSpring=true
    bOverrideAngularDamping=true
    bUseGravityOverride=true
    gravityScale=0.35
    simulationSpace=World
    alphaBoolBlend {blendOption=Linear}
    boundBone {boneName=cape_05_r}
    componentAppliedLinearAccClamp=<3000 3000 3000>
    componentLinearAccScale=<2 2 2>
    componentLinearVelScale=<2 2 2>
    gravityOverride {z=-980}
    physicsBodyDefinitions [{
      boundBone {boneName=cape_05_r}
      boxExtents=<30 40 40>
      constraintSetup {angularLimitsMax=<0 5 15>, angularLimitsMin=<0 -5 -15>, angularTarget=<1 0 0>}
      localJointOffset=<4 0 0>
     }]
    relativeSpaceBone {boneName=backpack_root}
   }
   n79 = Anim Dynamics - Bone: cape_07_r(ComponentPose=@n78, Alpha=1, ExternalForce="0,0,0") {
    angularDampingOverride=1
    angularSpringConstant=10000
    bAngularSpring=true
    bOverrideAngularDamping=true
    bUseGravityOverride=true
    gravityScale=0.35
    simulationSpace=World
    alphaBoolBlend {blendOption=Linear}
    boundBone {boneName=cape_07_r}
    componentAppliedLinearAccClamp=<3000 3000 3000>
    componentLinearAccScale=<2 2 2>
    componentLinearVelScale=<2 2 2>
    gravityOverride {z=-980}
    physicsBodyDefinitions [{
      boundBone {boneName=cape_07_r}
      boxExtents=<30 40 40>
      constraintSetup {angularLimitsMax=<0 5 15>, angularLimitsMin=<0 -5 -15>, angularTarget=<1 0 0>}
      localJointOffset=<4 0 0>
     }]
    relativeSpaceBone {boneName=backpack_root}
   }
   n80 = Anim Dynamics - Bone: cape_03_l(ComponentPose=@n79, Alpha=1, ExternalForce="0,0,0") {
    angularDampingOverride=1
    angularSpringConstant=10000
    bAngularSpring=true
    bOverrideAngularDamping=true
    bUseGravityOverride=true
    gravityScale=0.35
    simulationSpace=World
    alphaBoolBlend {blendOption=Linear}
    boundBone {boneName=cape_03_l}
    componentAppliedLinearAccClamp=<3000 3000 3000>
    componentLinearAccScale=<2 2 2>
    componentLinearVelScale=<2 2 2>
    gravityOverride {z=-980}
    physicsBodyDefinitions [{
      boundBone {boneName=cape_03_l}
      boxExtents=<30 40 40>
      constraintSetup {angularLimitsMax=<0 5 15>, angularLimitsMin=<0 -5 -15>, angularTarget=<1 0 0>}
      localJointOffset=<4 0 0>
     }]
    relativeSpaceBone {boneName=backpack_root}
   }
   n81 = Anim Dynamics - Bone: cape_05_l(ComponentPose=@n80, Alpha=1, ExternalForce="0,0,0") {
    angularDampingOverride=1
    angularSpringConstant=10000
    bAngularSpring=true
    bOverrideAngularDamping=true
    bUseGravityOverride=true
    gravityScale=0.35
    simulationSpace=World
    alphaBoolBlend {blendOption=Linear}
    boundBone {boneName=cape_05_l}
    componentAppliedLinearAccClamp=<3000 3000 3000>
    componentLinearAccScale=<2 2 2>
    componentLinearVelScale=<2 2 2>
    gravityOverride {z=-980}
    physicsBodyDefinitions [{
      boundBone {boneName=cape_05_l}
      boxExtents=<30 40 40>
      constraintSetup {angularLimitsMax=<0 5 15>, angularLimitsMin=<0 -5 -15>, angularTarget=<1 0 0>}
      localJointOffset=<4 0 0>
     }]
    relativeSpaceBone {boneName=backpack_root}
   }
   n82 = Anim Dynamics - Bone: cape_07_l(ComponentPose=@n81, Alpha=1, ExternalForce="0,0,0") {
    angularDampingOverride=1
    angularSpringConstant=10000
    bAngularSpring=true
    bOverrideAngularDamping=true
    bUseGravityOverride=true
    gravityScale=0.35
    simulationSpace=World
    alphaBoolBlend {blendOption=Linear}
    boundBone {boneName=cape_07_l}
    componentAppliedLinearAccClamp=<3000 3000 3000>
    componentLinearAccScale=<2 2 2>
    componentLinearVelScale=<2 2 2>
    gravityOverride {z=-980}
    physicsBodyDefinitions [{
      boundBone {boneName=cape_07_l}
      boxExtents=<30 40 40>
      constraintSetup {angularLimitsMax=<0 5 15>, angularLimitsMin=<0 -5 -15>, angularTarget=<1 0 0>}
      localJointOffset=<4 0 0>
     }]
    relativeSpaceBone {boneName=backpack_root}
   }
   n83 = Get notEmoting() -> (notEmoting:bool)
   n84 = "Transform (Modify) Bone - Bone: cape_02_l"!(ComponentPose=@n82, bAlphaBoolEnabled:bool=@n83, Translation="0,0,0", Rotation="15,13,0", Scale="1,1,1") -> (Pose) {
    alphaInputType=Bool
    rotationMode=BMM_Additive
    rotationSpace=BCS_BoneSpace
    translationMode=BMM_Additive
    translationSpace=BCS_BoneSpace
    alphaBoolBlend {blendInTime=2, blendOption=Linear, blendOutTime=2}
    boneToModify {boneName=cape_02_l}
    rotation {pitch=15, yaw=13}
   }
   n85 = "Transform (Modify) Bone - Bone: cape_mid_02"!(ComponentPose=@n84, bAlphaBoolEnabled:bool=@n83, Translation="0,0,0", Rotation="0,13,0", Scale="1,1,1") -> (Pose) {
    alphaInputType=Bool
    rotationMode=BMM_Additive
    rotationSpace=BCS_BoneSpace
    translationMode=BMM_Additive
    translationSpace=BCS_BoneSpace
    alphaBoolBlend {blendInTime=2, blendOption=Linear, blendOutTime=2}
    boneToModify {boneName=cape_mid_02}
    rotation {yaw=13}
   }
   n86 = "Transform (Modify) Bone - Bone: cape_03_r"!(ComponentPose=@n85, bAlphaBoolEnabled:bool=@n83, Translation="0,0,0", Rotation="5,5,0", Scale="1,1,1") -> (Pose) {
    alphaInputType=Bool
    rotationMode=BMM_Additive
    rotationSpace=BCS_BoneSpace
    translationMode=BMM_Additive
    translationSpace=BCS_BoneSpace
    alphaBoolBlend {blendInTime=2, blendOption=Linear, blendOutTime=2}
    boneToModify {boneName=cape_03_r}
    rotation {pitch=5, yaw=5}
   }
   Save cached pose 'animDyn Cache'(Pose=Blend Poses by bool(BlendPose_0=Component To Local(ComponentPose=@n86), BlendPose_1=Component To Local(ComponentPose=@n71), bActiveValue=Get bIsInFrontEnd()) {blendType=Linear}) {cacheName=animDyn Cache}
   Get Is Crouching()
   Get Is Sky Diving() -> (Is Sky Diving:double)
   Get Is Parachute Opened() -> (Is Parachute Opened:double)
  ]}
 EventGraph {nodes [
   n1 = PlayerMovement(AnimInstance=Self-Reference(), XDotProductCoefficient=1, YDotProductCoefficient=1, ZDotProductCoefficient=1)
   > n2 = Branch(Condition=Get bIsDBNOCarried())
   > Set BwdMovement(BwdMovement:double=0.5) -> (Output_Get:double)
   n3 = Branch(execute=@n2.else, Condition=AND Boolean(A=Get bIsTargeting(), B=Get bIsCrouching()))
   n4 = MapRangeClamped(
    Value=float + float(A="Absolute (Float)"(A=@n1), B=MapRangeClamped(Value=@n1.YMovement, InRangeA=-1, InRangeB=1, OutRangeA=0.3)) {is=title K2Node_CommutativeAssociativeBinaryOperator}
    InRangeA=0
    InRangeB=1.65
   )
   > Set BwdMovement(BwdMovement=float * float(A=@n4, B=0.25) {is=title K2Node_PromotableOperator})
   Set BwdMovement(execute=@n3.else, BwdMovement=@n4)
   Event BlueprintUpdateAnimation()
   > n5 = Sequence!(execute) -> (
    then_0
    then_1
    then_2:Exec
    then_3:Exec
    then_4:Exec
    then_5:Exec
    then_6:Exec
    then_7:Exec
    then_8:Exec
    then_9:Exec
    then_10:Exec
    then_11:Exec
    then_12:Exec
    then_13:Exec
    then_14:Exec
    then_15:Exec
    then_16:Exec
    then_17:Exec
    then_18:Exec
    then_19:Exec
    then_20:Exec
    then_21:Exec
    then_22:Exec
    then_23:Exec
    then_24:Exec
    then_25:Exec
   )
   n6 = GetOwnerSkeletalMeshComponent()
   > Set clavLf(clavLf:double=GetAlphaFromReferenceTransformDelta(SkeletalMeshComponent=@n6, BoneName=clavicle_l, BaseName=spine_05, DeltaComponentType=EulerY, InputRange=(X=20,Y=100), TwistAxis="1,0,0")) -> (Output_Get:double)
   > Set clavRt(clavRt:double=GetAlphaFromReferenceTransformDelta(SkeletalMeshComponent=@n6, BoneName=clavicle_r, BaseName=spine_05, InputRange=(X=20,Y=100), TwistAxis="1,0,0")) -> (Output_Get:double)
   n7 = GetOwnerSkeletalMeshComponent()
   > Set legSide(legSide:double=GetAlphaFromReferenceTransformDelta(SkeletalMeshComponent=@n7, BoneName=thigh_r, BaseName=pelvis, InputRange=(X=0,Y=-100), TwistAxis="1,0,0")) -> (Output_Get:double)
   > Set LegFwd(LegFwd:double=GetAlphaFromReferenceTransformDelta(SkeletalMeshComponent=@n7, BoneName=thigh_r, BaseName=pelvis, DeltaComponentType=EulerY, InputRange=(X=0,Y=-100), TwistAxis="1,0,0")) -> (Output_Get:double)
   > Set Is Crouching(Is Crouching:double="To Float (Boolean)"(InBool=Get bIsCrouching())) -> (Output_Get:double)
   > Set HeadRotBk(HeadRotBk:double=GetAlphaFromReferenceTransformDelta(
     SkeletalMeshComponent=GetOwnerSkeletalMeshComponent()
     BoneName=head
     BaseName=spine_05
     InputRange=(X=-20,Y=0)
     OutputRange=(X=1,Y=0)
     TwistAxis="1,0,0"
    )) -> (Output_Get:double)
   > Set HeadRotFr(HeadRotFr:double=GetAlphaFromReferenceTransformDelta(
     SkeletalMeshComponent=GetOwnerSkeletalMeshComponent()
     BoneName=head
     BaseName=spine_05
     InputRange=(X=20,Y=0)
     OutputRange=(X=1,Y=0)
     TwistAxis="1,0,0"
    )) -> (Output_Get:double)
   n8 = GetOwnerSkeletalMeshComponent()
   Set Clavicle LRot(
    execute=@n5.then_1
    Clavicle LRot:double=GetAlphaFromReferenceTransformDelta(SkeletalMeshComponent=@n8, BoneName=clavicle_l, BaseName=spine_05, DeltaComponentType=EulerY, InputRange=(X=20,Y=100), TwistAxis="1,0,0")
   ) -> (Output_Get:double)
   > Set Clavicle RRot(Clavicle RRot:double=GetAlphaFromReferenceTransformDelta(SkeletalMeshComponent=@n8, BoneName=clavicle_r, BaseName=spine_05, InputRange=(X=20,Y=100), TwistAxis="1,0,0")) -> (Output_Get:double)
   n9 = GetAnimInstance(self=GetOwnerSkeletalMeshComponent())
   Is Valid(exec=@n5.then_10, InputObject=@n9)
   > n10 = Set Is Custom Emote Playing(Is Custom Emote Playing:double=SelectFloat(A=1, B=0, bPickA=Montage_IsPlaying(self=@n9, Montage=/BRCosmetics/Animation/Game/MainPlayer/Emotes/Kitty_Cat/Emote_Kitty_Cat_CMF_M))) -> (Output_Get:double)
   > Set Custom Emote Not Playing(Custom Emote Not Playing:double="To Float (Boolean)"(InBool="float < float"(A=@n10, B=0.5))) -> (Output_Get:double)
   n11 = Get Custom Emote Not Playing() -> (Custom Emote Not Playing:double)
   Set isNotSkydivingNoBBNoEmote(execute=@n5.then_11, isNotSkydivingNoBBNoEmote:double=float * float(A=Get isNotSkydivingNoBB() -> (isNotSkydivingNoBB:double), B=@n11)) -> (Output_Get:double)
   Set isSkyDivingNoBBNoEmote(execute=@n5.then_12, isSkyDivingNoBBNoEmote:double=float * float(A=@n11, B=Get isSkyDivingNoBB())) -> (Output_Get:double)
   Set isWearingBBNoEmote(execute=@n5.then_13, isWearingBBNoEmote:double=float * float(A=@n11, B=Get isWearingBackBling() -> (isWearingBackBling:double))) -> (Output_Get:double)
   Set IsWearingBBNoEmoteIsSkydiving(
    execute=@n5.then_14
    IsWearingBBNoEmoteIsSkydiving:double=float * float(A=Get isWearingBBNoEmote() -> (isWearingBBNoEmote:double), B=Get Is Sky Diving())
   ) -> (Output_Get:double)
   Set isWearingBBIsNotSkydiving(
    execute=@n5.then_15
    isWearingBBIsNotSkydiving:double=float * float(A=Get isWearingBackBling(), B=Get isNotSkydiving() -> (isNotSkydiving:double))
   ) -> (Output_Get:double)
   > n12 = Set Is Parachute Opened(Is Parachute Opened:double="To Float (Boolean)"(InBool=Get bIsParachuteOpened())) -> (Output_Get:double)
   > Set isParachuteOpenedNoBB(isParachuteOpenedNoBB:double=float * float(A=@n12, B=Get isWearingBackBling())) -> (Output_Get:double)
   Set Is In Front End(execute=@n5.then_16, Is In Front End:double="To Float (Boolean)"(InBool=Get bIsInFrontEnd())) -> (Output_Get:double)
   n13 = Get bIsPlayingEmote()
   Set isNotEmoting(execute=@n5.then_17, isNotEmoting:double="To Float (Boolean)"(InBool=NOT Boolean(A=@n13))) -> (Output_Get:double)
   n14 = Get Is In Front End() -> (Is In Front End:double)
   Set FrontendNoEmote(execute=@n5.then_18, FrontendNoEmote:double=float * float(A=Get isNotEmoting() -> (isNotEmoting:double), B=@n14)) -> (Output_Get:double)
   n15 = Get bIsInFrontEnd()
   Set isNotInFrontend(execute=@n5.then_19, isNotInFrontend:double="To Float (Boolean)"(InBool=NOT Boolean(A=@n15))) -> (Output_Get:double)
   n16 = Set Is Sky Diving(execute=@n5.then_2, Is Sky Diving:double="To Float (Boolean)"(InBool=Get bIsSkyDiving())) -> (Output_Get:double)
   n17 = Get isNotWearingBB() -> (isNotWearingBB:double)
   > Set isSkyDivingNoBB(isSkyDivingNoBB:double=float * float(A=@n16, B=@n17)) -> (Output_Get:double)
   n18 = Get isNotInFrontend() -> (isNotInFrontend:double)
   Set IsNotSkydivingNoBBNoFE(execute=@n5.then_20, IsNotSkydivingNoBBNoFE:double=float * float(A=Get isNotSkydivingNoBB(), B=@n18)) -> (Output_Get:double)
   > Set isSkyDivingNoBBNoEmoteNoFE(isSkyDivingNoBBNoEmoteNoFE:double=float * float(A=Get isSkyDivingNoBBNoEmote() -> (isSkyDivingNoBBNoEmote:double), B=@n18)) -> (Output_Get:double)
   > Set IsWearingBBNoFE(IsWearingBBNoFE:double=float * float(A=Get isWearingBackBling(), B=@n18)) -> (Output_Get:double)
   > Set isWearingBBNoEmoteIsSkydivingNoFE(isWearingBBNoEmoteIsSkydivingNoFE:double=float * float(A=Get IsWearingBBNoEmoteIsSkydiving() -> (IsWearingBBNoEmoteIsSkydiving:double), B=@n18)) -> (Output_Get:double)
   n19 = Set Is Playing Emote(execute=@n5.then_21, Is Playing Emote:double="To Float (Boolean)"(InBool=Get bIsPlayingEmote())) -> (Output_Get:double)
   > Set isPlayingEmoteFE(isPlayingEmoteFE:double=float * float(A=@n14, B=@n19)) -> (Output_Get:double)
   n20 = Get bRigidBodyEnabled()
   Set Rigid Body Not Enabled(execute=@n5.then_22, Rigid Body Not Enabled:double="To Float (Boolean)"(InBool=NOT Boolean(A=@n20))) -> (Output_Get:double)
   > Set RigidBodyOffSkydiving(RigidBodyOffSkydiving:double=float * float(A=Get Rigid Body Not Enabled() -> (Rigid Body Not Enabled:double), B=Get Is Sky Diving())) -> (Output_Get:double)
   n21 = Gravity Override State Function DEPRECATED(
    execute=@n5.then_23
    CharacterPartAnimInstance=Self-Reference()
    GameTimeInSeconds=GetGameTimeInSeconds()
    GravityOverrideParams=Get Gravity Override Params()
    DebugJointName=head
   )
   > Set GravityOverride(GravityOverride=@n21)
   Set NoBBNoFE(
    execute=@n5.then_24
    NoBBNoFE:double="Max (Float)"(A=Get IsNotSkydivingNoBBNoFE() -> (IsNotSkydivingNoBBNoFE:double), B=Get isSkyDivingNoBBNoEmoteNoFE() -> (isSkyDivingNoBBNoEmoteNoFE:double))
   ) -> (Output_Get:double)
   Set wearing backpack and skydiving(execute=@n5.then_25, wearing backpack and skydiving:double=float * float(A=Get isWearingBackBling(), B=Get Is Sky Diving())) -> (Output_Get:double)
   > Set notEmoting(notEmoting:bool=NOT Boolean(A=Get bIsPlayingEmote())) -> (Output_Get:bool)
   n22 = Get bIsSkyDiving()
   n23 = Set isNotSkydiving(execute=@n5.then_3, isNotSkydiving:double="To Float (Boolean)"(InBool=NOT Boolean(A=@n22))) -> (Output_Get:double)
   > Set isNotSkydivingNoBB(isNotSkydivingNoBB:double=float * float(A=@n17, B=@n23)) -> (Output_Get:double)
   n24 = GetOwnerSkeletalMeshComponent()
   Set ClavFwdLf(
    execute=@n5.then_4
    ClavFwdLf:double=GetAlphaFromReferenceTransformDelta(
     SkeletalMeshComponent=@n24
     BoneName=clavicle_l
     BaseName=spine_05
     DeltaComponentType=EulerY
     InputRange=(X=20,Y=0)
     OutputRange=(X=1,Y=0)
     TwistAxis="1,0,0"
    )
   ) -> (Output_Get:double)
   > Set ClavFwdRt(ClavFwdRt:double=GetAlphaFromReferenceTransformDelta(SkeletalMeshComponent=@n24, BoneName=clavicle_r, BaseName=spine_05, InputRange=(X=20,Y=0), OutputRange=(X=1,Y=0), TwistAxis="1,0,0")) -> (Output_Get:double)
   n25 = Set Is Skydive Floating(execute=@n5.then_5, Is Skydive Floating:double="To Float (Boolean)"(InBool=Get bIsSkydiveFloating())) -> (Output_Get:double)
   > Set isSkydiveFloatingNoBB(isSkydiveFloatingNoBB:double=float * float(A=@n25, B=Get isNotWearingBB())) -> (Output_Get:double)
   n26 = GetAnimInstance(self=GetOwnerSkeletalMeshComponent())
   Is Valid(exec=@n5.then_6, InputObject=@n26)
   > Set Is Custom Any Emote Playing(Is Custom Any Emote Playing:double=SelectFloat(A=1, B=0, bPickA=IsAnyMontagePlaying!(self:Anim Instance Object Reference=@n26) -> (ReturnValue:bool))) -> (Output_Get:double)
   n27 = Get bIsSkydiveFloating()
   Set isNotSkydiveFloating(execute=@n5.then_7, isNotSkydiveFloating:double="To Float (Boolean)"(InBool=NOT Boolean(A=@n27))) -> (Output_Get:double)
   Set IsSkydivingNotFloating(
    execute=@n5.then_8
    IsSkydivingNotFloating:double=float * float(A=Get Is Sky Diving(), B=Get isNotSkydiveFloating() -> (isNotSkydiveFloating:double))
   ) -> (Output_Get:double)
   n28 = Set Is Skydive Dive Mode(execute=@n5.then_9, Is Skydive Dive Mode:double="To Float (Boolean)"(InBool=Get bIsSkydiveDiveMode())) -> (Output_Get:double)
   > Set isSkydiveDiveNoBB(isSkydiveDiveNoBB:double=float * float(A=@n28, B=Get isNotWearingBB())) -> (Output_Get:double)
   Event OnCharacterCustomizationLoaded()
   > Do Once()
   n29 = CheckForBackbling()
   > Set isWearingBackBling(isWearingBackBling:double="To Float (Boolean)"(InBool=OR Boolean(A=@n29, B=@n29.bIsBackblingLong))) -> (Output_Get:double)
   > Do Once()
   n30 = CheckForBackbling()
   n31 = OR Boolean(A=@n30, B=@n30.bIsBackblingLong)
   > Set isNotWearingBB(isNotWearingBB:double="To Float (Boolean)"(InBool=NOT Boolean(A=@n31))) -> (Output_Get:double)
   > n32 = Branch(Condition=CharacterItemDefinitionHasAnyMetaTag(Tags=Make Array("[0]"=(TagName="Cosmetics.Character.RestrictCape"))))
   > Set cape_active(cape_active:bool=true) -> (Output_Get:bool)
   Set cape_active(execute=@n32.else, cape_active=false)
   Event OnBeginSkydiving()
   > Set SkydivingAlpha(SkydivingAlpha:Rotator="65,0,0") -> (Output_Get:Rotator)
   Event OnEndSkydiving()
   > Set SkydivingAlpha(SkydivingAlpha="10,0,0")
   n33 = Branch(Condition=Get bIsSkyDiving())
   > n34 = Branch(Condition="float > float"(A=Get PawnVelocityZ(), B=50) {is=title K2Node_CallFunction})
   > Set BwdMovement(BwdMovement=float * float(A=Get BwdMovement() -> (BwdMovement:double), B=0) {is=title K2Node_CommutativeAssociativeBinaryOperator})
   Set BwdMovement(execute=[@n33.else, @n34.else], BwdMovement=float * float(A=Get BwdMovement(), B=1))
   Get isWearingBBIsNotSkydiving() -> (isWearingBBIsNotSkydiving:double)
  ]}
]
variables [
 BwdMovement {type=number}
 cape_active {type=boolean}
 ClavFwdLf {type=number}
 ClavFwdRt {type=number}
 Clavicle LRot {type=number}
 Clavicle RRot {type=number}
 clavLf {type=number}
 clavRt {type=number}
 Custom Emote Not Playing {type=number}
 FrontendNoEmote {type=number}
 Gravity Override Cape {type=Vector}
 Gravity Override Params {type=GravityOverrideParamsStruct}
 GravityOverride {type=Vector}
 GravityOverride_cape {type=Vector}
 HeadRotBk {type=number}
 HeadRotFr {type=number}
 Is Crouching {type=number}
 Is Custom Any Emote Playing {type=number}
 Is Custom Emote Playing {type=number}
 Is In Front End {type=number}
 Is Parachute Opened {type=number}
 Is Playing Emote {type=number}
 Is Sky Diving {type=number}
 Is Skydive Dive Mode {type=number}
 Is Skydive Floating {type=number}
 isNotEmoting {type=number}
 isNotInFrontend {type=number}
 isNotSkydiveFloating {type=number}
 isNotSkydiving {type=number}
 isNotSkydivingNoBB {type=number}
 isNotSkydivingNoBBNoEmote {type=number}
 isNotSkydivingNoBBNoEmoteInterp {type=number}
 IsNotSkydivingNoBBNoFE {type=number}
 isNotWearingBB {type=number}
 isParachuteOpenedNoBB {type=number}
 isPlayingEmoteFE {type=number}
 isSkydiveDiveNoBB {type=number}
 isSkydiveFloatingNoBB {type=number}
 isSkyDivingNoBB {type=number}
 isSkyDivingNoBBNoEmote {type=number}
 isSkyDivingNoBBNoEmoteNoFE {type=number}
 IsSkydivingNotFloating {type=number}
 isWearingBackBling {type=number}
 isWearingBBDiveMode {type=number}
 isWearingBBIsNotSkydiving {type=number}
 isWearingBBNoEmote {type=number}
 IsWearingBBNoEmoteIsSkydiving {type=number}
 isWearingBBNoEmoteIsSkydivingNoFE {type=number}
 IsWearingBBNoFE {type=number}
 LegFwd {type=number}
 legSide {type=number}
 NoBBNoFE {type=number}
 notEmoting {type=boolean}
 RegularAndFloating {type=number}
 Rigid Body Not Enabled {type=number}
 RigidBodyOffSkydiving {type=number}
 SkydivingAlpha {type=Rotator}
 wearing backpack and skydiving {type=number}
]
