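# Animation Blueprint for the face accessory part of the M_MED_Sienna player cosmetic, with a bone adjustment that applies only in the front end.
# The AnimGraph copies the pose from the attached parent mesh, then additively shifts deform_trap_r along its negative local Z.
# Each animation update converts the front-end flag to a float that drives the Transform Bone alpha, switching the offset on or off.
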AnimBlueprint /BRCosmetics/Characters/Player/Male/Medium/Bodies/M_MED_Sienna/Meshes/Parts/M_MED_Sienna_FaceAcc_AnimBP
target_skeleton=&/M_MED_Sienna_FaceAcc_Skeleton
graphs [
 AnimGraph {nodes [
   n1 = "Transform (Modify) Bone - Bone: deform_trap_r"(
    ComponentPose=Local To Component(LocalPose=Copy Pose From Mesh() {bUseAttachedParent=true})
    Alpha=Get Is In Front End() -> (Is In Front End:double)
    Translation="0,0,-3"
    Rotation="0,0,0"
    Scale="1,1,1"
   ) {
    rotationMode=BMM_Additive
    rotationSpace=BCS_BoneSpace
    scaleMode=BMM_Replace
    scaleSpace=BCS_BoneSpace
    translationMode=BMM_Additive
    translationSpace=BCS_BoneSpace
    alphaBoolBlend {blendOption=Linear}
    boneToModify {boneName=deform_trap_r}
    translation {z=-3}
   }
   Output Pose(Result=Component To Local(ComponentPose=@n1)) {layerGroup=None}
  ]}
 EventGraph {nodes [
   Event BlueprintUpdateAnimation()
   > Set Is In Front End(Is In Front End:double="To Float (Boolean)"(InBool=Get bIsInFrontEnd())) -> (Output_Get:double)
  ]}
]
variables [Is In Front End {type=number}]
